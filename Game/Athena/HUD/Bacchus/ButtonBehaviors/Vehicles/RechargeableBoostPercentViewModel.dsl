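; Progress bar view model that tracks a vehicle's rechargeable boost percent for a HUD button behavior.
; Initialization caches pawn and controller, runs the enter handler once, and binds driver-enter and vehicle-exit events, which attach or detach the boost-change delegate.
; The bar enables only for rechargeable-boost vehicles matching a gameplay tag query, then mirrors boost changes.
; ShouldBecomeVisibleImmediately returns the current percent and whether it is below full.
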(event EventInitialize
  (params (Pawn :type Pawn%20Object%20Reference))
  (|CallParentFunction Pawn)
  (bind _asfort_player_pawn (Utilities|Casting|CastToFortPlayerPawn Pawn))
  (bind _output_get (Variables|Default|SetFort%20Player%20Pawn _asfort_player_pawn))
  (bind _asfort_player_controller (Utilities|Casting|CastToFortPlayerController (Pawn|GetController _output_get)))
  (bind _output_get_2 (Variables|Default|SetFort%20Player%20Controller _asfort_player_controller))
  (Vehicle|BindEventtoOnEnterVehicleDriver _output_get_2 (EventDispatchers|CreateEvent :!event Custom|Event%20On%20Enter%20Vehicle))
  (CallFunction|EventonEnterVehicle)
  (Vehicle|BindEventtoOnExitVehicle _output_get_2 (EventDispatchers|CreateEvent :!event Custom|Event%20On%20Exit%20Vehicle)))

(event EventUninitialize
  (params (Pawn :type Pawn%20Object%20Reference))
  (bind _fort_player_controller (Variables|Default|GetFort%20Player%20Controller))
  (Utilities|IsValid _fort_player_controller
    (:Is%20Valid
      (Vehicle|UnbindEventfromOnEnterVehicleDriver _fort_player_controller (EventDispatchers|CreateEvent :!event Custom|Event%20On%20Enter%20Vehicle))
      (Vehicle|UnbindEventfromOnExitVehicle _fort_player_controller (EventDispatchers|CreateEvent :!event Custom|Event%20On%20Exit%20Vehicle)))))

(event Custom|Event%20On%20Enter%20Vehicle
  (bind _output_get (Variables|Default|SetFort%20Athena%20Vehicle (Vehicle|GetVehicle (Variables|Default|GetFort%20Player%20Pawn))))
  (Vehicle|Boost|BindEventtoOnRechargeableBoostPercentChangedDelegate _output_get (EventDispatchers|CreateEvent :!event OnRechargeableBoostPercentChanged))
  (Variables|GenericProgressBarViewModel|SetbEnabled (CallFunction|CanEnable)))

(event Custom|Event%20On%20Exit%20Vehicle
  (bind _fort_athena_vehicle (Variables|Default|GetFort%20Athena%20Vehicle))
  (Utilities|IsValid _fort_athena_vehicle
    (:Is%20Valid
      (Vehicle|Boost|UnbindEventfromOnRechargeableBoostPercentChangedDelegate _fort_athena_vehicle (EventDispatchers|CreateEvent :!event OnRechargeableBoostPercentChanged))
      (Variables|Default|SetFort%20Athena%20Vehicle 0)
      (Variables|GenericProgressBarViewModel|SetbEnabled false))))

(fn CanEnable
  (bind _fort_athena_vehicle (Variables|Default|GetFort%20Athena%20Vehicle))
  (Utilities|IsValid _fort_athena_vehicle
    (:Is%20Valid
      (return (and (Class|FortVehicleInterface|DoesVehicleUseRechargeableBoost _fort_athena_vehicle true) (GameplayTags|DoesContainerMatchTagQuery (Class|FortVehicleInterface|GetGameplayTags _fort_athena_vehicle) (Variables|Default|GetVehicleGameplayTagsQuery)))))
    (:Is%20Not%20Valid
      (return false))))

(fn OnRechargeableBoostPercentChanged
  (params (CurrentValue :type Float%20%28single-precision%29) (PreviousValue :type Float%20%28single-precision%29))
  (if (Variables|GenericProgressBarViewModel|GetbEnabled)
    (Variables|GenericProgressBarViewModel|SetProgressValue CurrentValue)))

(fn ShouldBecomeVisibleImmediately
  (bind _fort_athena_vehicle (Variables|Default|GetFort%20Athena%20Vehicle))
  (if (CallFunction|CanEnable)
    (Utilities|IsValid _fort_athena_vehicle
      (:Is%20Valid
        (return (Class|FortVehicleInterface|GetRechargeableBoostPercent _fort_athena_vehicle) (< (Class|FortVehicleInterface|GetRechargeableBoostPercent _fort_athena_vehicle) 1.0)))
      (:Is%20Not%20Valid
        _))
    (else
      _))
  (return 0.0 false))
