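; Vehicle HUD progress bar view model, a boosting variant of the rechargeable boost percent model, enabled only while the vehicle is boosting.
; On entering a vehicle it binds to the vehicle's boost started and finished delegates, unbinding them on exit.
; Boost start sets the enabled flag from CanEnable, and boost finish clears it.
; The CanEnable override requires a valid vehicle, the parent's check and IsBoosting.
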(event EventEventOnEnterVehicle
  (bind _fort_athena_vehicle (Variables|Default|GetFort%20Athena%20Vehicle))
  (|CallParentFunction)
  (Utilities|IsValid _fort_athena_vehicle
    (:Is%20Valid
      (Vehicle|Boost|BindEventtoOnBoostStartedDelegate _fort_athena_vehicle (EventDispatchers|CreateEvent :!event Custom|Event%20On%20Boost%20Started))
      (Vehicle|Boost|BindEventtoOnBoostFinishedDelegate _fort_athena_vehicle (EventDispatchers|CreateEvent :!event Custom|Event%20On%20Boost%20Finished)))))

(event EventEventOnExitVehicle
  (bind _fort_athena_vehicle (Variables|Default|GetFort%20Athena%20Vehicle))
  (Utilities|IsValid _fort_athena_vehicle
    (:Is%20Valid
      (Vehicle|Boost|UnbindEventfromOnBoostStartedDelegate _fort_athena_vehicle (EventDispatchers|CreateEvent :!event Custom|Event%20On%20Boost%20Started))
      (Vehicle|Boost|UnbindEventfromOnBoostFinishedDelegate _fort_athena_vehicle (EventDispatchers|CreateEvent :!event Custom|Event%20On%20Boost%20Finished))
      (|CallParentFunction))))

(event Custom|Event%20On%20Boost%20Started
  (Variables|GenericProgressBarViewModel|SetbEnabled (CallFunction|CanEnable)))

(event Custom|Event%20On%20Boost%20Finished
  (Variables|GenericProgressBarViewModel|SetbEnabled false))

(fn CanEnable
  (bind _fort_athena_vehicle (Variables|Default|GetFort%20Athena%20Vehicle))
  (Utilities|IsValid _fort_athena_vehicle
    (:Is%20Valid
      (return (and (|CallParentFunction) (Class|FortVehicleInterface|IsBoosting _fort_athena_vehicle))))
    (:Is%20Not%20Valid
      (return false))))
